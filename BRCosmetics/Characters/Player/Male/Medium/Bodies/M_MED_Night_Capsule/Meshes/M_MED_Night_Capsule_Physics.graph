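# Physics asset for the M_MED_Night_Capsule male player body, covering the dangling skirt panels and holster with leg and pelvis collision shapes.
# Skirt chains and the holster hang from the pelvis through limited joints, with secondary D_a and H_a branches off D02 and H02.
# Panels mostly use boxes, often paired at angles to follow curvature.
# Free-angular cross constraints tie neighbouring lower panels together like a cloth lattice, leaving chain F unlinked.
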PhysicsAsset /BRCosmetics/Characters/Player/Male/Medium/Bodies/M_MED_Night_Capsule/Meshes/M_MED_Night_Capsule_Physics
bodies {
 calf_l [{shapeType=Capsule, length=43.033, radius=7.5105, shapeName=calf_l_capsule, center=<-22.8318 0 1.3971>, rotation=<rot 89.6654 -0.0002 0>}]
 calf_r [{shapeType=Capsule, length=43.5907, radius=6.9528, shapeName=calf_l_capsule, center=<23.1527 0 -0.064>, rotation=<rot 89.6654 -0.0002 0>}]
 dyn_Holster [{
   shapeType=Capsule
   length=16.7342
   radius=5.1086
   shapeName=dyn_Holster_capsule
   center=<-2.4872 3.074 -7.4995>
   rotation=<rot -6.1225 -90.0111 1.8308>
  }]
 dyn_Holster_base [{shapeType=Capsule, length=0.505, radius=0.505, shapeName=dyn_Holster_base_capsule, rotation=<rot 0 -90 0>}]
 dyn_SkirtB01 [{shapeType=Capsule, length=0.505, radius=0.505, shapeName=dyn_SkirtB01_capsule, center=<-0.0001 0 2.8265>, rotation=<rot 0 -89.9999 0>}]
 dyn_SkirtB02 [{shapeType=Capsule, length=15.2562, radius=1.859, shapeName=dyn_SkirtB02_capsule, center=<-0.0002 0 -8.9789>}]
 dyn_SkirtB03 [{
   shapeType=Box
   extentX=22.5789
   extentY=1.5
   extentZ=21.9489
   shapeName=dyn_SkirtB03_box
   center=<0.9438 0.0001 -8.9344>
   rotation=<rot 0 -20 0>
  }]
 dyn_SkirtB04 [{
   shapeType=Box
   extentX=22.5789
   extentY=1.5
   extentZ=21.9489
   shapeName=dyn_SkirtB04_box
   center=<0.9438 0.0001 -8.9344>
   rotation=<rot 0 -20 0>
  }]
 dyn_SkirtD01 [{shapeType=Capsule, length=0.505, radius=0.505, shapeName=dyn_SkirtD01_capsule, rotation=<rot 0 -90 0>}]
 dyn_SkirtD02 [{shapeType=Capsule, length=13.8118, radius=3.0664, shapeName=dyn_SkirtD02_capsule, center=<0 0.0002 -7.7703>}]
 dyn_SkirtD03 [
  {
   shapeType=Box
   extentX=15.9007
   extentY=2.7875
   extentZ=23.5866
   shapeName=dyn_SkirtD03_box
   center=<-6.9095 -4.3094 -8.9345>
   rotation=<rot 0 39.9994 0>
  }
  {
   shapeType=Box
   extentX=19.9007
   extentY=2.7875
   extentZ=23.5866
   shapeName=dyn_SkirtD03_box
   center=<5.0384 2.4094 -9.7172>
   rotation=<rot 0 20 0>
  }
 ]
 dyn_SkirtD04 [
  {
   shapeType=Box
   extentX=15.9007
   extentY=2.7875
   extentZ=22.2864
   shapeName=dyn_SkirtD04_box
   center=<-5.9025 -3.9429 -9.1587>
   rotation=<rot 0 39.9992 0>
  }
  {
   shapeType=Box
   extentX=17.1933
   extentY=2.7875
   extentZ=22.2864
   shapeName=dyn_SkirtD04_box
   center=<5.4908 2.5741 -9.1262>
   rotation=<rot 0 20 0>
  }
 ]
 dyn_SkirtD_a03 [
  {
   shapeType=Box
   extentX=15.1484
   extentY=2.221
   extentZ=21.0467
   shapeName=dyn_SkirtD_a03_box
   center=<-5.4612 0.6029 -8.1457>
   rotation=<rot 0 -5.4924 0>
  }
  {
   shapeType=Box
   extentX=14.7285
   extentY=2.221
   extentZ=21.0467
   shapeName=dyn_SkirtD_a03_box
   center=<4.0311 -2.5749 -8.1455>
   rotation=<rot 0 -32.1776 0>
  }
 ]
 dyn_SkirtD_a03_base [{shapeType=Capsule, length=5.081, radius=4.0343, shapeName=dyn_SkirtD_a03_base_capsule, rotation=<rot 0 -90 0>}]
 dyn_SkirtD_a04 [
  {
   shapeType=Box
   extentX=17.8772
   extentY=2.5685
   extentZ=22.1163
   shapeName=dyn_SkirtD_a04_box
   center=<-5.8646 1.0766 -9.216>
   rotation=<rot 0 -5.4924 0>
  }
  {
   shapeType=Box
   extentX=16.4964
   extentY=2.5685
   extentZ=22.1163
   shapeName=dyn_SkirtD_a04_box
   center=<4.9941 -1.8343 -9.2159>
   rotation=<rot 0 -32.1779 0>
  }
 ]
 dyn_SkirtE01 [{shapeType=Capsule, length=0.505, radius=0.505, shapeName=dyn_SkirtE01_capsule, rotation=<rot 0 -90 0>}]
 dyn_SkirtE02 [{shapeType=Capsule, length=13.8009, radius=4.4992, shapeName=dyn_SkirtE02_capsule, center=<0 0 -9.2781>}]
 dyn_SkirtE03 [
  {
   shapeType=Box
   extentX=14.2905
   extentY=4.8093
   extentZ=20.9946
   shapeName=dyn_SkirtE03_box
   center=<-4.9221 -1.1344 -8.9605>
   rotation=<rot 0 19.9996 0>
  }
  {shapeType=Box, extentX=14.2905, extentY=4.8093, extentZ=21.1264, shapeName=dyn_SkirtE03_box, center=<6.4654 0.6725 -8.9607>}
 ]
 dyn_SkirtE04 [
  {
   shapeType=Box
   extentX=14.2905
   extentY=5.009
   extentZ=20.4944
   shapeName=dyn_SkirtE04_box
   center=<-4.9221 -1.1344 -8.2063>
   rotation=<rot 0 19.9996 0>
  }
  {shapeType=Box, extentX=14.2905, extentY=5.009, extentZ=20.6262, shapeName=dyn_SkirtE04_box, center=<6.4654 0.6725 -8.2064>}
 ]
 dyn_SkirtF01 [{shapeType=Capsule, length=0.505, radius=0.505, shapeName=dyn_SkirtF01_capsule, rotation=<rot 0 -90 0>}]
 dyn_SkirtF02 [{shapeType=Capsule, length=14.5475, radius=1.7362, shapeName=dyn_SkirtF02_capsule, center=<0 0 -7.3818>}]
 dyn_SkirtF03 [{shapeType=Capsule, length=14.5475, radius=1.7362, shapeName=dyn_SkirtF03_capsule, center=<0 0 -7.3818>}]
 dyn_SkirtG01 [{shapeType=Capsule, length=0.505, radius=0.505, shapeName=dyn_SkirtE01_capsule, rotation=<rot 0 -90 0>}]
 dyn_SkirtG02 [{shapeType=Capsule, length=13.8009, radius=4.4992, shapeName=dyn_SkirtE02_capsule, center=<0 0 -9.2781>}]
 dyn_SkirtG03 [
  {
   shapeType=Box
   extentX=13
   extentY=2.5
   extentZ=20.5
   shapeName=dyn_SkirtG03_box
   center=<-5.1215 -0.8104 -9.1406>
   rotation=<rot 0 10.0022 0>
  }
  {shapeType=Box, extentX=13, extentY=2.5, extentZ=20.5, shapeName=dyn_SkirtG03_box, center=<5.4037 0.274 -9.1406>, rotation=<rot 0 -0.0011 0>}
 ]
 dyn_SkirtG04 [
  {
   shapeType=Box
   extentX=13.5
   extentY=3.5
   extentZ=20.5
   shapeName=dyn_SkirtG04_box
   center=<-4.43 -0.7695 -8.6611>
   rotation=<rot 0 10.0036 0>
  }
  {
   shapeType=Box
   extentX=13.5
   extentY=3.5
   extentZ=20.5
   shapeName=dyn_SkirtG04_box
   center=<5.1359 0.1942 -8.6613>
   rotation=<rot 0 -0.0011 0>
  }
 ]
 dyn_SkirtG_base [{shapeType=Box, extentX=15.8097, extentY=5.4625, extentZ=4.3525, shapeName=dyn_SkirtG_base_box, center=<0.0504 -0.0001 -0.0208>}]
 dyn_SkirtH01 [{shapeType=Capsule, length=0.505, radius=0.505, shapeName=dyn_SkirtD01_capsule, rotation=<rot 0 -90 0>}]
 dyn_SkirtH02 [{shapeType=Capsule, length=13.8118, radius=3.0664, shapeName=dyn_SkirtD02_capsule, center=<0 0.0002 -7.7703>}]
 dyn_SkirtH03 [
  {
   shapeType=Box
   extentX=15.1169
   extentY=3.7875
   extentZ=23.5866
   shapeName=dyn_SkirtD03_box
   center=<-4.1642 3.233 -8.9346>
   rotation=<rot 0 -20.0003 0>
  }
  {
   shapeType=Box
   extentX=15.6299
   extentY=3.7875
   extentZ=23.5866
   shapeName=dyn_SkirtD03_box
   center=<5.2636 -3.583 -8.6091>
   rotation=<rot 0 -50.0004 0>
  }
 ]
 dyn_SkirtH04 [
  {
   shapeType=Box
   extentX=15.4708
   extentY=3.7875
   extentZ=23.5866
   shapeName=dyn_SkirtD04_box
   center=<-5.3063 1.7079 -8.9345>
   rotation=<rot 0 -20.0007 0>
  }
  {
   shapeType=Box
   extentX=15.4441
   extentY=3.7875
   extentZ=23.5866
   shapeName=dyn_SkirtD04_box
   center=<3.5684 -3.4041 -9.7172>
   rotation=<rot 0 -40.3149 0>
  }
 ]
 dyn_SkirtH_a03 [
  {
   shapeType=Box
   extentX=12.036
   extentY=2.3387
   extentZ=22.8435
   shapeName=dyn_SkirtH_a03_box
   center=<-3.9988 -0.5496 -8.5443>
   rotation=<rot 0 20.0008 0>
  }
  {
   shapeType=Box
   extentX=11.7926
   extentY=3.8387
   extentZ=22.8435
   shapeName=dyn_SkirtH_a03_box
   center=<4.4556 0.688 -8.5442>
   rotation=<rot 0 10.001 0>
  }
 ]
 dyn_SkirtH_a03_base [{shapeType=Capsule, length=0, radius=5.6024, shapeName=dyn_SkirtH_a03_base_capsule, rotation=<rot 0 -90 0>}]
 dyn_SkirtH_a04 [
  {
   shapeType=Box
   extentX=13.3276
   extentY=2.6369
   extentZ=20.5384
   shapeName=dyn_SkirtH_a04_box
   center=<-3.5976 -2.3662 -7.3365>
   rotation=<rot 0 30.001 0>
  }
  {
   shapeType=Box
   extentX=12.6102
   extentY=3.1369
   extentZ=20.5384
   shapeName=dyn_SkirtH_a04_box
   center=<4.5691 0.3769 -7.3363>
   rotation=<rot 0 10.0018 0>
  }
 ]
 dyn_SkirtJ01 [{shapeType=Capsule, length=0.505, radius=0.505, shapeName=dyn_SkirtJ01_capsule, rotation=<rot 0 -90 0>}]
 dyn_SkirtJ02 [{shapeType=Capsule, length=15.2562, radius=1.859, shapeName=dyn_SkirtB02_capsule, center=<-0.0002 0 -8.9789>}]
 dyn_SkirtJ03 [{
   shapeType=Box
   extentX=22.5789
   extentY=1.5
   extentZ=21.9489
   shapeName=dyn_SkirtB03_box
   center=<-0.8883 -0.2042 -8.9343>
   rotation=<rot 0 27.137 0>
  }]
 dyn_SkirtJ04 [{
   shapeType=Box
   extentX=22.5789
   extentY=1.5
   extentZ=21.9489
   shapeName=dyn_SkirtB04_box
   center=<-0.8883 -0.2042 -8.9343>
   rotation=<rot 0 27.137 0>
  }]
 foot_l [{shapeType=Capsule, length=15.8308, radius=7.7742, shapeName=foot_l_capsule, rotation=<rot 0 0 -90.0004>}]
 foot_r [{shapeType=Capsule, length=17.5925, radius=8.9236, shapeName=foot_r_capsule, center=<0.1911 -2.8095 0>, rotation=<rot 0 0 -90.0004>}]
 pelvis [{shapeType=Capsule, length=10.1003, radius=8.0603, shapeName=pelvis_capsule, center=<-5.4132 -1.4168 0.0001>}]
 thigh_l [{
   shapeType=Capsule
   length=27.3004
   radius=6.7404
   shapeName=thigh_l_capsule
   center=<-20.9081 0 0.1453>
   rotation=<rot 89.6019 0.0056 0.0056>
  }]
 thigh_r [{
   shapeType=Capsule
   length=27.5646
   radius=6.4762
   shapeName=thigh_l_capsule
   center=<21.3254 -0.0001 -0.1482>
   rotation=<rot 89.6019 0.0056 0.0056>
  }]
 thigh_twist_01_l []
 thigh_twist_01_r []
}
constraints [
 {bone1Name=dyn_SkirtJ04, bone2Name=dyn_SkirtJ03, swing1LimitDegrees=15, swing2LimitDegrees=15, twistLimitDegrees=45}
 {bone1Name=dyn_SkirtJ03, bone2Name=dyn_SkirtJ02, swing1LimitDegrees=15, swing2LimitDegrees=15, twistLimitDegrees=45}
 {bone1Name=dyn_SkirtJ02, bone2Name=dyn_SkirtJ01, swing1LimitDegrees=0.1, swing2LimitDegrees=0.1, twistLimitDegrees=2}
 {bone1Name=dyn_SkirtH04, bone2Name=dyn_SkirtH03, swing1LimitDegrees=60, swing2LimitDegrees=60, twistLimitDegrees=60}
 {bone1Name=dyn_SkirtH03, bone2Name=dyn_SkirtH02, swing1LimitDegrees=60, swing2LimitDegrees=60, twistLimitDegrees=60}
 {bone1Name=dyn_SkirtH02, bone2Name=dyn_SkirtH01, swing1LimitDegrees=15, swing2LimitDegrees=15, twistLimitDegrees=45}
 {bone1Name=dyn_SkirtG04, bone2Name=dyn_SkirtG03, swing1LimitDegrees=30, swing2LimitDegrees=30, twistLimitDegrees=60}
 {bone1Name=dyn_SkirtG03, bone2Name=dyn_SkirtG02, swing1LimitDegrees=30, swing2LimitDegrees=30, twistLimitDegrees=60}
 {bone1Name=dyn_SkirtG02, bone2Name=dyn_SkirtG01, swing1LimitDegrees=0.1, swing2LimitDegrees=0.1, twistLimitDegrees=2}
 {bone1Name=dyn_SkirtF03, bone2Name=dyn_SkirtF02, swing1LimitDegrees=15, swing2LimitDegrees=15, twistLimitDegrees=45}
 {bone1Name=dyn_SkirtF02, bone2Name=dyn_SkirtF01, swing1LimitDegrees=15, swing2LimitDegrees=15, twistLimitDegrees=45}
 {bone1Name=dyn_SkirtE04, bone2Name=dyn_SkirtE03, swing1LimitDegrees=30, swing2LimitDegrees=30, twistLimitDegrees=60}
 {bone1Name=dyn_SkirtE03, bone2Name=dyn_SkirtE02, swing1LimitDegrees=30, swing2LimitDegrees=30, twistLimitDegrees=60}
 {bone1Name=dyn_SkirtE02, bone2Name=dyn_SkirtE01, swing1LimitDegrees=1, swing2LimitDegrees=1, twistLimitDegrees=1}
 {bone1Name=dyn_SkirtD04, bone2Name=dyn_SkirtD03, swing1LimitDegrees=90, swing2LimitDegrees=90, twistLimitDegrees=90}
 {bone1Name=dyn_SkirtD03, bone2Name=dyn_SkirtD02, swing1LimitDegrees=90, swing2LimitDegrees=90, twistLimitDegrees=90}
 {bone1Name=dyn_SkirtD02, bone2Name=dyn_SkirtD01, swing1LimitDegrees=15, swing2LimitDegrees=15, twistLimitDegrees=20}
 {bone1Name=dyn_SkirtB04, bone2Name=dyn_SkirtB03, swing1LimitDegrees=20, swing2LimitDegrees=30, twistLimitDegrees=30}
 {bone1Name=dyn_SkirtB03, bone2Name=dyn_SkirtB02, swing1LimitDegrees=30, swing2LimitDegrees=30, twistLimitDegrees=30}
 {bone1Name=dyn_SkirtB02, bone2Name=dyn_SkirtB01, swing1LimitDegrees=5, swing2LimitDegrees=5, twistLimitDegrees=30}
 {bone1Name=dyn_SkirtJ01, bone2Name=pelvis, swing1LimitDegrees=1, swing2LimitDegrees=1, twistLimitDegrees=45}
 {bone1Name=dyn_SkirtG01, bone2Name=pelvis, swing1LimitDegrees=1, swing2LimitDegrees=1, twistLimitDegrees=45}
 {bone1Name=dyn_SkirtH01, bone2Name=pelvis, swing1LimitDegrees=1, swing2LimitDegrees=1, twistLimitDegrees=45}
 {bone1Name=dyn_SkirtF01, bone2Name=pelvis, swing1LimitDegrees=1, swing2LimitDegrees=1, twistLimitDegrees=45}
 {bone1Name=dyn_SkirtE01, bone2Name=pelvis, swing1LimitDegrees=1, swing2LimitDegrees=1, twistLimitDegrees=45}
 {bone1Name=dyn_SkirtD01, bone2Name=pelvis, swing1LimitDegrees=1, swing2LimitDegrees=1, twistLimitDegrees=45}
 {bone1Name=dyn_SkirtB01, bone2Name=pelvis, swing1LimitDegrees=1, swing2LimitDegrees=1, twistLimitDegrees=45}
 {bone1Name=dyn_Holster_base, bone2Name=pelvis, swing1LimitDegrees=45, swing2LimitDegrees=45, twistLimitDegrees=45}
 {bone1Name=dyn_Holster, bone2Name=dyn_Holster_base, swing1LimitDegrees=7, swing2LimitDegrees=7, twistLimitDegrees=40}
 {bone1Name=dyn_SkirtH_a04, bone2Name=dyn_SkirtH_a03, swing1LimitDegrees=60, swing2LimitDegrees=60, twistLimitDegrees=60}
 {bone1Name=dyn_SkirtH_a03, bone2Name=dyn_SkirtH_a03_base, swing1LimitDegrees=60, swing2LimitDegrees=60, twistLimitDegrees=60}
 {bone1Name=dyn_SkirtD_a04, bone2Name=dyn_SkirtD_a03, swing1LimitDegrees=60, swing2LimitDegrees=60, twistLimitDegrees=60}
 {bone1Name=dyn_SkirtD_a03, bone2Name=dyn_SkirtD_a03_base, swing1LimitDegrees=60, swing2LimitDegrees=10, twistLimitDegrees=60}
 {bone1Name=dyn_SkirtH_a03_base, bone2Name=dyn_SkirtH02, swing1LimitDegrees=1, swing2LimitDegrees=1, twistLimitDegrees=1}
 {bone1Name=dyn_SkirtD_a03_base, bone2Name=dyn_SkirtD02, swing1LimitDegrees=1, swing2LimitDegrees=1, twistLimitDegrees=1}
 {
  bone1Name=dyn_SkirtH_a04
  bone2Name=dyn_SkirtH04
  swing1LimitDegrees=45
  swing1Motion=Free
  swing2LimitDegrees=45
  swing2Motion=Free
  twistLimitDegrees=45
  twistMotion=Free
 }
 {
  bone1Name=dyn_SkirtD_a03
  bone2Name=dyn_SkirtD03
  swing1LimitDegrees=45
  swing1Motion=Free
  swing2LimitDegrees=45
  swing2Motion=Free
  twistLimitDegrees=45
  twistMotion=Free
 }
 {
  bone1Name=dyn_SkirtD_a04
  bone2Name=dyn_SkirtD04
  swing1LimitDegrees=45
  swing1Motion=Free
  swing2LimitDegrees=45
  swing2Motion=Free
  twistLimitDegrees=45
  twistMotion=Free
 }
 {
  bone1Name=dyn_SkirtD04
  bone2Name=dyn_SkirtB04
  swing1LimitDegrees=45
  swing1Motion=Free
  swing2LimitDegrees=45
  swing2Motion=Free
  twistLimitDegrees=45
  twistMotion=Free
 }
 {
  bone1Name=dyn_SkirtH_a03
  bone2Name=dyn_SkirtH03
  swing1LimitDegrees=45
  swing1Motion=Free
  swing2LimitDegrees=45
  swing2Motion=Free
  twistLimitDegrees=45
  twistMotion=Free
 }
 {
  bone1Name=dyn_SkirtH04
  bone2Name=dyn_SkirtJ04
  swing1LimitDegrees=45
  swing1Motion=Free
  swing2LimitDegrees=45
  swing2Motion=Free
  twistLimitDegrees=45
  twistMotion=Free
 }
 {
  bone1Name=dyn_SkirtH_a04
  bone2Name=dyn_SkirtG04
  swing1LimitDegrees=45
  swing1Motion=Free
  swing2LimitDegrees=45
  swing2Motion=Free
  twistLimitDegrees=45
  twistMotion=Free
 }
 {
  bone1Name=dyn_SkirtH03
  bone2Name=dyn_SkirtG03
  swing1LimitDegrees=45
  swing1Motion=Free
  swing2LimitDegrees=45
  swing2Motion=Free
  twistLimitDegrees=45
  twistMotion=Free
 }
 {
  bone1Name=dyn_SkirtH03
  bone2Name=dyn_SkirtJ03
  swing1LimitDegrees=45
  swing1Motion=Free
  swing2LimitDegrees=45
  swing2Motion=Free
  twistLimitDegrees=45
  twistMotion=Free
 }
 {
  bone1Name=dyn_SkirtD_a04
  bone2Name=dyn_SkirtE04
  swing1LimitDegrees=45
  swing1Motion=Free
  swing2LimitDegrees=45
  swing2Motion=Free
  twistLimitDegrees=45
  twistMotion=Free
 }
 {
  bone1Name=dyn_SkirtD_a03
  bone2Name=dyn_SkirtE03
  swing1LimitDegrees=45
  swing1Motion=Free
  swing2LimitDegrees=45
  swing2Motion=Free
  twistLimitDegrees=45
  twistMotion=Free
 }
]
