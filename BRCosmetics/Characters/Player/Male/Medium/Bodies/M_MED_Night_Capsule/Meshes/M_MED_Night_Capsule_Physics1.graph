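# Physics asset for the M_MED_Night_Capsule male medium character body, providing capsule bodies for a dynamic cloth strip on the left forearm.
# A three-segment dyn_cloth_l chain hangs from a base bone constrained to lowerarm_l, each link with free swing on both axes.
# Capsules on lowerarm_l and hand_l are joined by a fully locked constraint.
# A dyn_Holster_base capsule is also present, with no constraint.
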PhysicsAsset /BRCosmetics/Characters/Player/Male/Medium/Bodies/M_MED_Night_Capsule/Meshes/M_MED_Night_Capsule_Physics1
bodies {
 dyn_Holster_base [{shapeType=Capsule, length=21.1479, radius=8.8521, shapeName=dyn_Holster_base_capsule, center=<-0.0001 0 -14.1097>}]
 dyn_cloth_01_l [{
   shapeType=Capsule
   length=2.8957
   radius=1.5521
   shapeName=dyn_cloth_01_l_capsule
   center=<-1.6765 0 0.0147>
   rotation=<rot -89.4964 -179.9964 179.9965>
  }]
 dyn_cloth_01_l_base [{shapeType=Capsule, length=1.5859, radius=1.5521, shapeName=dyn_cloth_01_l_base_capsule, rotation=<rot 0 -89.9999 90.0555>}]
 dyn_cloth_02_l [{
   shapeType=Capsule
   length=4.0867
   radius=1.5521
   shapeName=dyn_cloth_02_l_capsule
   center=<-1.6765 0 0.0147>
   rotation=<rot -89.4964 -179.9964 179.9965>
  }]
 dyn_cloth_03_l [{
   shapeType=Capsule
   length=4.0867
   radius=1.5521
   shapeName=dyn_cloth_03_l_capsule
   center=<-1.6765 0 0.0147>
   rotation=<rot -89.4964 -179.9964 179.9965>
  }]
 hand_l [{
   shapeType=Capsule
   length=9.1162
   radius=5.8464
   shapeName=hand_l_capsule
   center=<3.9955 0.0001 0.0815>
   rotation=<rot -88.8315 -0.002 0.002>
  }]
 lowerarm_l [{
   shapeType=Capsule
   length=17.8684
   radius=3.7449
   shapeName=lowerarm_l_capsule
   center=<10.8833 -0.0002 -0.091>
   rotation=<rot -89.5211 179.9962 -179.9962>
  }]
}
constraints [
 {
  bone1Name=hand_l
  bone2Name=lowerarm_l
  swing1LimitDegrees=45
  swing1Motion=Locked
  swing2LimitDegrees=45
  swing2Motion=Locked
  twistLimitDegrees=45
  twistMotion=Locked
 }
 {
  bone1Name=dyn_cloth_01_l
  bone2Name=dyn_cloth_01_l_base
  swing1LimitDegrees=90
  swing1Motion=Free
  swing2LimitDegrees=90
  swing2Motion=Free
  twistLimitDegrees=45
 }
 {
  bone1Name=dyn_cloth_02_l
  bone2Name=dyn_cloth_01_l
  swing1LimitDegrees=90
  swing1Motion=Free
  swing2LimitDegrees=90
  swing2Motion=Free
  twistLimitDegrees=45
 }
 {
  bone1Name=dyn_cloth_03_l
  bone2Name=dyn_cloth_02_l
  swing1LimitDegrees=90
  swing1Motion=Free
  swing2LimitDegrees=90
  swing2Motion=Free
  twistLimitDegrees=45
 }
 {
  bone1Name=dyn_cloth_01_l_base
  bone2Name=lowerarm_l
  swing1LimitDegrees=90
  swing1Motion=Free
  swing2LimitDegrees=90
  swing2Motion=Free
  twistLimitDegrees=45
 }
]
