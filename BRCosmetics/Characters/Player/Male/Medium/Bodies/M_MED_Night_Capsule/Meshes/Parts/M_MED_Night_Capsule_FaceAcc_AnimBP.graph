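# Animation blueprint for the Night Capsule male medium body's face accessory mesh, adding secondary jiggle to four dyn_hant bones.
# It copies the pose from the attached parent mesh, then chains one Anim Dynamics node per bone, a through d.
# Each node runs at half alpha in root-relative space with identical damped angular springs, tight twist limits, low gravity and reduced component motion influence.
# The event graph has no logic.
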AnimBlueprint /BRCosmetics/Characters/Player/Male/Medium/Bodies/M_MED_Night_Capsule/Meshes/Parts/M_MED_Night_Capsule_FaceAcc_AnimBP
target_skeleton=&/M_MED_Night_Capsule_FaceAcc_Skeleton
graphs [
 AnimGraph {nodes [
   n1 = Anim Dynamics - Bone: dyn_hant_a(ComponentPose=Local To Component(LocalPose=Copy Pose From Mesh() {bUseAttachedParent=true}), Alpha=0.5, ExternalForce="0,0,0") {
    angularDampingOverride=0.999
    angularSpringConstant=800
    bAngularSpring=true
    bGravityOverrideInSimSpace=true
    bOverrideAngularDamping=true
    gravityScale=0.2
    simulationSpace=RootRelative
    alphaBoolBlend {blendOption=Linear}
    alphaScaleBiasClamp {scale=0.25}
    boundBone {boneName=dyn_hant_a}
    componentAppliedLinearAccClamp=<10000 10000 10000>
    componentLinearAccScale=<0.1 0.1 0.1>
    componentLinearVelScale=<0.1 0.1 0.1>
    physicsBodyDefinitions [{
      sphereCollisionRadius=0
      boundBone {boneName=dyn_hant_a}
      boxExtents=<5 10 10>
      constraintSetup {angularTargetAxis=AxisY, twistAxis=AxisZ, angularLimitsMax=<2 30 30>, angularLimitsMin=<-2 -30 -30>, angularTarget=<0 0 0>}
      localJointOffset=<-3 -0.1 -0.1>
     }]
    relativeSpaceBone {boneName=spine_05}
   }
   n2 = Anim Dynamics - Bone: dyn_hant_b(ComponentPose=@n1, Alpha=0.5, ExternalForce="0,0,0") {
    angularDampingOverride=0.999
    angularSpringConstant=800
    bAngularSpring=true
    bGravityOverrideInSimSpace=true
    bOverrideAngularDamping=true
    gravityScale=0.2
    simulationSpace=RootRelative
    alphaBoolBlend {blendOption=Linear}
    alphaScaleBiasClamp {scale=0.25}
    boundBone {boneName=dyn_hant_b}
    componentAppliedLinearAccClamp=<10000 10000 10000>
    componentLinearAccScale=<0.1 0.1 0.1>
    componentLinearVelScale=<0.1 0.1 0.1>
    physicsBodyDefinitions [{
      sphereCollisionRadius=0
      boundBone {boneName=dyn_hant_b}
      boxExtents=<5 10 10>
      constraintSetup {angularTargetAxis=AxisY, twistAxis=AxisZ, angularLimitsMax=<2 30 30>, angularLimitsMin=<-2 -30 -30>, angularTarget=<0 0 0>}
      localJointOffset=<-3 -0.1 -0.1>
     }]
    relativeSpaceBone {boneName=spine_05}
   }
   n3 = Anim Dynamics - Bone: dyn_hant_c(ComponentPose=@n2, Alpha=0.5, ExternalForce="0,0,0") {
    angularDampingOverride=0.999
    angularSpringConstant=800
    bAngularSpring=true
    bGravityOverrideInSimSpace=true
    bOverrideAngularDamping=true
    gravityScale=0.2
    simulationSpace=RootRelative
    alphaBoolBlend {blendOption=Linear}
    alphaScaleBiasClamp {scale=0.25}
    boundBone {boneName=dyn_hant_c}
    componentAppliedLinearAccClamp=<10000 10000 10000>
    componentLinearAccScale=<0.1 0.1 0.1>
    componentLinearVelScale=<0.1 0.1 0.1>
    physicsBodyDefinitions [{
      sphereCollisionRadius=0
      boundBone {boneName=dyn_hant_c}
      boxExtents=<5 10 10>
      constraintSetup {angularTargetAxis=AxisY, twistAxis=AxisZ, angularLimitsMax=<2 30 30>, angularLimitsMin=<-2 -30 -30>, angularTarget=<0 0 0>}
      localJointOffset=<-3 -0.1 -0.1>
     }]
    relativeSpaceBone {boneName=spine_05}
   }
   n4 = Anim Dynamics - Bone: dyn_hant_d(ComponentPose=@n3, Alpha=0.5, ExternalForce="0,0,0") {
    angularDampingOverride=0.999
    angularSpringConstant=800
    bAngularSpring=true
    bGravityOverrideInSimSpace=true
    bOverrideAngularDamping=true
    gravityScale=0.2
    simulationSpace=RootRelative
    alphaBoolBlend {blendOption=Linear}
    alphaScaleBiasClamp {scale=0.25}
    boundBone {boneName=dyn_hant_d}
    componentAppliedLinearAccClamp=<10000 10000 10000>
    componentLinearAccScale=<0.1 0.1 0.1>
    componentLinearVelScale=<0.1 0.1 0.1>
    physicsBodyDefinitions [{
      sphereCollisionRadius=0
      boundBone {boneName=dyn_hant_d}
      boxExtents=<5 10 10>
      constraintSetup {angularTargetAxis=AxisY, twistAxis=AxisZ, angularLimitsMax=<2 30 30>, angularLimitsMin=<-2 -30 -30>, angularTarget=<0 0 0>}
      localJointOffset=<-3 -0.1 -0.1>
     }]
    relativeSpaceBone {boneName=spine_05}
   }
   Output Pose(Result=Component To Local(ComponentPose=@n4)) {layerGroup=None}
  ]}
 EventGraph {nodes [Event BlueprintUpdateAnimation()]}
]
variables []
